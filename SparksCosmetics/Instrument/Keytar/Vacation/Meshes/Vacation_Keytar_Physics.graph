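# Physics asset for the Vacation Keytar instrument cosmetic, defining bodies and joints for its dangling tag and two wheels, apparently for secondary motion.
# A large capsule on the handle anchors the rig, with offset spheres on the left and right wheels.
# The tag is a two-link chain from handle through a sphere to a capsule.
# All three joints attached to the handle lock twist, leaving movement to swing.
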PhysicsAsset /SparksCosmetics/Instrument/Keytar/Vacation/Meshes/Vacation_Keytar_Physics
bodies {
 dyn_tag_01 [{shapeType=Sphere, radius=2, shapeName=dyn_tag_01_sphere}]
 dyn_tag_03 [{shapeType=Capsule, length=4, radius=2.5, center=<4.3201 0.0016 -0.1449>, rotation=<rot -89.7324 167.7827 -167.7629>}]
 dyn_wheel_left [{shapeType=Sphere, radius=3.0419, shapeName=dyn_wheel_left_sphere, center=<0 0 -5>}]
 dyn_wheel_right [{shapeType=Sphere, radius=2.9077, shapeName=dyn_wheel_right_sphere, center=<0 0 -5>}]
 handle [{shapeType=Capsule, length=35.7163, radius=13.0218, center=<-0.2404 -10.1159 7.9992>, rotation=<rot 1.1554 -90 1.7753>}]
}
constraints [
 {bone1Name=dyn_tag_03, bone2Name=dyn_tag_01, swing1LimitDegrees=40, swing2LimitDegrees=55, twistLimitDegrees=15}
 {bone1Name=dyn_tag_01, bone2Name=handle, swing1LimitDegrees=1, swing2LimitDegrees=50, twistLimitDegrees=5, twistMotion=Locked}
 {bone1Name=dyn_wheel_right, bone2Name=handle, swing1LimitDegrees=30, swing2LimitDegrees=1, twistLimitDegrees=45, twistMotion=Locked}
 {bone1Name=dyn_wheel_left, bone2Name=handle, swing1LimitDegrees=30, swing2LimitDegrees=1, twistLimitDegrees=45, twistMotion=Locked}
]
